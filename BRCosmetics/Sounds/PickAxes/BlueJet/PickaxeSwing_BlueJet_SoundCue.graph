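# Sound cue for the BlueJet cosmetic pickaxe swing in Battle Royale, with close-range melee weapon attenuation.
# A switch on the SwingIdx parameter sends most inputs to a shared random pick of swing waves.
# Its fifth input instead layers a random flourish over that swing pick, with the swing mixed lower.
# The result runs through modulator and attenuation nodes under player melee concurrency.
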SoundCue /BRCosmetics/Sounds/PickAxes/BlueJet/PickaxeSwing_BlueJet_SoundCue
root=Mixer_0
nodes [
 {id=Attenuation_0, children [Modulator_0], properties {attenuationSettings=/Game/Sounds/Attenuation/Fort_Weapon_Pickaxe_Close_Att}}
 {id=Mixer_0, children [Attenuation_0], properties {inputVolume [1]}}
 {id=Mixer_1, children [Random_1,Random_0], properties {inputVolume [1,0.6]}}
 {id=Modulator_0, children [Switch_1]}
 {id=Random_0, children [WavePlayer_0,WavePlayer_1,WavePlayer_2,WavePlayer_3,WavePlayer_4,WavePlayer_5], properties {weights [1,1,1,1,1,1]}}
 {id=Random_1, children [WavePlayer_6,WavePlayer_7,WavePlayer_8,WavePlayer_9], properties {weights [1,1,1,1]}}
 {id=Switch_1, children [Random_0,Random_0,Random_0,Random_0,Mixer_1,Random_0], properties {intParameterName=SwingIdx}}
 {id=WavePlayer_0, properties {soundWaveAssetPtr=&/Pickaxe_BlueJet_Swings_01}}
 {id=WavePlayer_1, properties {soundWaveAssetPtr=&/Pickaxe_BlueJet_Swings_02}}
 {id=WavePlayer_2, properties {soundWaveAssetPtr=&/Pickaxe_BlueJet_Swings_03}}
 {id=WavePlayer_3, properties {soundWaveAssetPtr=&/Pickaxe_BlueJet_Swings_04}}
 {id=WavePlayer_4, properties {soundWaveAssetPtr=&/Pickaxe_BlueJet_Swings_05}}
 {id=WavePlayer_5, properties {soundWaveAssetPtr=&/Pickaxe_BlueJet_Swings_06}}
 {id=WavePlayer_6, properties {soundWaveAssetPtr=&/Pickaxe_BlueJet_Flourish_01}}
 {id=WavePlayer_7, properties {soundWaveAssetPtr=&/Pickaxe_BlueJet_Flourish_02}}
 {id=WavePlayer_8, properties {soundWaveAssetPtr=&/Pickaxe_BlueJet_Flourish_03}}
 {id=WavePlayer_9, properties {soundWaveAssetPtr=&/Pickaxe_BlueJet_Flourish_04}}
]
properties {
 audioPropertiesSheet=/BRCosmetics/Sounds/Settings/PropertySheets/Pickaxe/APS_Cosmetic_Pickaxe_Swing
 maxDistance=1536
 soundClassObject=&Fort_WeaponMelee_PSM
 volumeMultiplier=0.3
 concurrencySet [&Player_Melee_Weapon_Conc]
}
